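# Material function for RustBucket NitroDrome screens, setting emissive, roughness and specular, with either a countdown timer or flipbook reactions.
# The timer picks digits from custom primitive data via an If chain and samples a digit strip with explicit derivatives.
# Reactions use primitive-data index and start time to drive curve-atlas colors and the switch between reaction and default textures.
# LCD pixel quantization, scanlines and periodic glitching style the image, with eye-adaptation-adjusted emissive.
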MaterialFunction /Game/Environments/Helios/Props/RustBucket/RustBucket_Screens_A/Materials/MF_NitroDromeScreen
graph {nodes [
  n1 = SetMaterialAttributes!(
   MaterialAttributes=FunctionInput() {inputType=FunctionInput_MaterialAttributes}
   BaseColor=Constant3Vector()
   EmissiveColor=NamedRerouteUsage()
   Roughness=ScalarParameter() {defaultValue=0.2, group=NitroDromeScreens, parameterName=Screen_Roughness}
   Specular=ScalarParameter() {defaultValue=0.5, group=NitroDromeScreens, parameterName=Screen_Specular}
  ) {attributeSetTypes [
    #1
    #2
    #3
    #4
   ]}
  n2 = MaterialFunctionCall!(Emissive, Alpha=Constant() {r=1}, Material=@n1) {materialFunction=/Game/Packages/Fortress_SharedMaterials/Base_Material_Functions/BaseCharacterShaderFunctions/MF_CosmeticEmissiveEyeAdaptation}
  FunctionOutput(None=@n2.Material)
  n3 = Multiply(A=StaticSwitch(True=NamedRerouteUsage(), False=NamedRerouteUsage(), Value=NamedRerouteUsage()), B=NamedRerouteUsage())
  Multiply(A=@n3, B=NamedRerouteUsage())
  n4 = NamedRerouteUsage()
  n5 = NamedRerouteDeclaration(None=ScalarParameter() {defaultValue=5, group=NitroDromeScreens, parameterName=Timer_NumberOfDigits, sortPriority=200}) {name=NumDigits, nodeColor=<rgba 1 0 0.0703 1>}
  n6 = Multiply(A=NamedRerouteDeclaration(None=ComponentMask(None=@n4) {r=true}) {name=U, nodeColor=<rgba 0 1 0.9766 1>}, B=@n5) {constB=5}
  n7 = Constant() {r=6}
  n8 = Subtract(
   A=NamedRerouteDeclaration(None=ScalarParameter() {defaultValue=11, group=NitroDromeScreens, parameterName=Timer_TextureDigits, sortPriority=200}) {name=TextureDigits, nodeColor=<rgba 1 0.6563 0 1>}
   B=Subtract(A=@n5, B=@n6)
  )
  n9 = ComponentMask(None=@n8) {r=true}
  n10 = VectorParameter() {bUseCustomPrimitiveData=true, group=NitroDromeScreens, parameterName=Timer_TimeRemaining, sortPriority=200, defaultValue=<rgba 1 2 3 4>}
  n11 = If(A=Constant() {r=6}, B=@n9, "A > B"=Constant() {r=-1}, "A < B"=@n10.R)
  n12 = Constant() {r=9}
  n13 = If(A=Constant() {r=7}, B=@n9, "A > B"=@n11, "A < B"=@n10.G)
  n14 = If(A=Constant() {r=8}, B=@n9, "A > B"=@n13, "A < B"=Constant() {r=-1})
  n15 = If(A=Constant() {r=9}, B=@n9, "A > B"=@n14, "A < B"=@n10.B)
  n16 = If(A=Constant() {r=10}, B=@n9, "A > B"=@n15, "A < B"=@n10.A)
  n17 = If(A=@n8, B=Constant() {r=10}, "A > B"=If(A=@n8, B=Constant() {r=11}, "A > B"=@n16, "A == B"=@n16, "A < B"=@n16), "A == B"=@n15, "A < B"=@n15)
  n18 = If(
   A=@n8
   B=Constant() {r=7}
   "A > B"=If(A=@n8, B=Constant() {r=8}, "A > B"=If(A=@n8, B=@n12, "A > B"=@n17, "A == B"=@n14, "A < B"=@n14), "A == B"=@n13, "A < B"=@n13)
   "A == B"=@n11
   "A < B"=@n11
  )
  n19 = Constant() {r=-1}
  n20 = Add(A=Frac(None=@n6), B=Floor(None=If(A=@n8, B=@n7, "A > B"=@n18, "A == B"=@n19, "A < B"=@n19)))
  n21 = NamedRerouteUsage()
  n22 = TextureSampleParameter2D!(
   UVs=AppendVector(A=Divide(A=@n20, B=@n21) {constB=11}, B=NamedRerouteUsage())
   "DDX(UVs)"=NamedRerouteUsage()
   "DDY(UVs)"=NamedRerouteUsage()
   Apply View MipBias
  ) {
   group=NitroDromeScreens
   mipValueMode=TMVM_Derivative
   parameterName=Timer_NumbersTexture
   samplerType=SAMPLERTYPE_Alpha
   sortPriority=200
   texture=/Game/Environments/Helios/Props/RustBucket/RustBucket_Screens_A/Textures/Screens/T_RustBucket_NitroDromeCenter_ScreenDigits
  }
  NamedRerouteDeclaration(None=Multiply(A=@n22.R, B=NamedRerouteUsage())) {name=CountdownNumber, nodeColor=<rgba 0.5 1 0 1>}
  n23 = NamedRerouteUsage()
  n24 = Multiply(A=OneMinus(None=ComponentMask(None=@n23) {g=true})) {constB=0.75}
  n25 = Multiply(A=Time(), B=ScalarParameter() {defaultValue=-0.2, group=NitroDromeScreens, parameterName=Sweep Speed}) {constB=-0.2}
  n26 = Power(Base=Frac(None=Add(A=@n24, B=@n25))) {constExponent=20}
  n27 = LinearInterpolate(
   A=ScalarParameter() {defaultValue=0.3, group=NitroDromeScreens, parameterName=Sweep Add Base}
   B=ScalarParameter() {defaultValue=0.3, group=NitroDromeScreens, parameterName=Sweep Add}
  ) {constAlpha=1}
  n28 = Multiply(A=@n26, B=ScalarParameter() {defaultValue=0.03, group=NitroDromeScreens, parameterName=Sweep UV Distort}) {constB=0.02}
  n29 = Constant()
  NamedRerouteDeclaration(None=Multiply(A=Multiply(A=@n26, B=@n27) {constB=0.2}, B=AppendVector(A=@n28, B=@n29))) {name=UV Distort, nodeColor=<rgba 0.0312 0 1 1>}
  n30 = NamedRerouteUsage()
  n31 = Add(
   A=StaticSwitch(True=Frac(None=@n30), False=NamedRerouteUsage(), Value=NamedRerouteUsage())
   B=AppendVector(
    A=ScalarParameter() {group=NitroDromeScreens, parameterName=Timer_NumberPaddingOffset_U, sortPriority=200}
    B=ScalarParameter() {group=NitroDromeScreens, parameterName=Timer_NumberPaddingOffset_V, sortPriority=200}
   )
  )
  n32 = AppendVector(
   A=ScalarParameter() {defaultValue=1, group=NitroDromeScreens, parameterName=Timer_NumberPaddingScale_U, sortPriority=200}
   B=ScalarParameter() {defaultValue=1, group=NitroDromeScreens, parameterName=Timer_NumberPaddingScale_V, sortPriority=200}
  )
  n33 = ComponentMask(None=VectorParameter() {parameterName=Bias, defaultValue=<rgba 0.02 0.04 0 1>}.RGB) {g=true, r=true}
  NamedRerouteDeclaration(None=MaterialFunctionCall(UVs=@n31, Texture Scale=Subtract(A=@n32, B=@n33)) {materialFunction=&ScaleUVsByCenter}.0-1 mask) {name=UVs_Scaled0-1Mask, nodeColor=<rgba 1 0.7969 0 1>}
  n34 = VectorParameter() {group=NitroDromeScreens, parameterName=Flipbook_Scale, sortPriority=301, defaultValue=<rgba 1 1 0 1>}
  n35 = VectorParameter() {group=NitroDromeScreens, parameterName=Flipbook_Offset, sortPriority=301, defaultValue=<rgba 0 0 0 1>}
  n36 = Add(A=Multiply(A=NamedRerouteUsage(), B=ComponentMask(None=@n34.RGB) {g=true, r=true}), B=ComponentMask(None=@n35.RGB) {g=true, r=true})
  n37 = CurveAtlasRowParameter(CurveTime=NamedRerouteUsage()) {
   atlas=/Game/Environments/Helios/Props/RustBucket/RustBucket_Screens_A/Textures/Screens/CA_NitroDromeScreen
   curve=/Game/Environments/Helios/Props/RustBucket/RustBucket_Screens_A/Textures/Screens/C_NitroDromeScreen
   group=NitroDromeScreens
   parameterName=ReactionDropAnim
  }
  n38 = Multiply(A=@n37.R, B=Constant2Vector() {g=1})
  n39 = ScalarParameter() {defaultValue=2, group=NitroDromeScreens, parameterName=Flipbook_Columns, sortPriority=300}
  n40 = AppendVector(A=@n39, B=ScalarParameter() {defaultValue=2, group=NitroDromeScreens, parameterName=Flipbook_Rows, sortPriority=300})
  n41 = Divide(A=Frac(None=Add(A=@n36, B=@n38)), B=@n40)
  n42 = Frac(None=Multiply(A=Time(), B=ScalarParameter() {defaultValue=1, group=NitroDromeScreens, parameterName=Flipbook_PlaySpeed, sortPriority=300}))
  n43 = ScalarParameter() {defaultValue=3, group=NitroDromeScreens, parameterName=Flipbook_TotalFrames, sortPriority=300}
  n44 = StaticSwitchParameter(True=ScalarParameter() {defaultValue=1, group=NitroDromeScreens, parameterName=DEBUG_FrameIndex}, False=Floor(None=Multiply(A=@n42, B=@n43))) {group=NitroDromeScreens, parameterName=DEBUG_Frame}
  n45 = Fmod(A=@n44, B=@n39)
  n46 = Divide(A=@n44, B=@n39)
  n47 = Divide(A=AppendVector(A=Floor(None=@n45), B=Floor(None=@n46)), B=@n40)
  NamedRerouteDeclaration(None=Frac(None=Add(A=@n41, B=@n47))) {name=FlipbookUV, nodeColor=<rgba 0 0.3906 1 1>}
  n48 = StaticSwitchParameter(
   True=ScalarParameter() {group=NitroDromeScreens, parameterName=DEBUG_ReactionIndex, sortPriority=300}
   False=ScalarParameter() {bUseCustomPrimitiveData=true, group=NitroDromeScreens, parameterName=ReactionIndex, primitiveDataIndex=2}
  ) {group=NitroDromeScreens, parameterName=DEBUG_Reaction}
  n49 = LinearInterpolate(
   A=Divide(A=Floor(None=@n48)) {constB=6}
   B=LinearInterpolate(A=Constant() {r=1}, B=Constant() {r=0.85}, Alpha=ScalarParameter() {group=NitroDromeScreens, parameterName=UseBossScreen})
   Alpha=Step(X=NamedRerouteUsage()) {constY=1}
  )
  n50 = CurveAtlasRowParameter(CurveTime=@n49) {
   atlas=&CA_NitroDromeScreen
   curve=/Game/Environments/Helios/Props/RustBucket/RustBucket_Screens_A/Textures/Screens/C_NitroDromeScreen_Colors1
   defaultValue=1
   group=NitroDromeScreens
   parameterName=Color1
  }
  n51 = CurveAtlasRowParameter(CurveTime=@n49) {
   atlas=&CA_NitroDromeScreen
   curve=/Game/Environments/Helios/Props/RustBucket/RustBucket_Screens_A/Textures/Screens/C_NitroDromeScreen_Colors2
   defaultValue=2
   group=NitroDromeScreens
   parameterName=Color2
  }
  n52 = NamedRerouteUsage()
  n53 = TextureObjectParameter() {group=NitroDromeScreens, parameterName=ReactionTexture, texture=&T_BlacksRGB}
  n54 = TextureSampleParameter2D!(UVs=NamedRerouteUsage(), Apply View MipBias) {
   parameterName=DefaultTexture
   samplerType=SAMPLERTYPE_Alpha
   texture=/Game/Environments/Helios/Props/RustBucket/RustBucket_Screens_A/Textures/Screens/T_NitroDromeScreen_Flipbook_Default
  }
  n55 = Constant()
  n56 = Subtract(A=Time(), B=ScalarParameter() {bUseCustomPrimitiveData=true, group=NitroDromeScreens, parameterName=GameTimeFromBP, primitiveDataIndex=1})
  n57 = Saturate(None=Divide(
    A=NamedRerouteDeclaration(None=@n56) {name=ReactionTimePassed, nodeColor=<rgba 0.4062 1 0 1>}
    B=ScalarParameter() {defaultValue=10, group=NitroDromeScreens, parameterName=Reaction_Duration, sortPriority=300}
   ))
  n58 = NamedRerouteDeclaration(None=LinearInterpolate(
    A=StaticSwitchParameter(True=@n55, False=@n57) {group=NitroDromeScreens, parameterName=DEBUG_Reaction}
    Alpha=ScalarParameter() {group=NitroDromeScreens, parameterName=UseBossScreen}
   )) {name=ReactionProgress, nodeColor=<rgba 0 1 0.9766 1>}
  n59 = Multiply(
   A=LinearInterpolate(
    A=TextureSample(UVs=@n52, Tex=@n53) {samplerSource=SSM_Clamp_WorldGroupSettings, samplerType=SAMPLERTYPE_Masks}.R
    B=@n54.R
    Alpha=Step(X=@n58) {constY=1}
   ) {constB=0}
   B=StaticSwitchParameter(True=Constant() {r=1}, False=NamedRerouteUsage()) {group=NitroDromeScreens, parameterName=DEBUG_Reaction}
  )
  NamedRerouteDeclaration(None=Multiply(
    A=LinearInterpolate(A=@n50, B=@n51, Alpha=@n59)
    B=ScalarParameter() {defaultValue=15, group=NitroDromeScreens, parameterName=Reaction_ScreenBrightness, sortPriority=300}
   )) {name=ReactionColor, nodeColor=<rgba 0.7109 1 0 1>}
  n60 = Add(
   A=MaterialFunctionCall!(
    Number of Scanlines=ScalarParameter() {defaultValue=50, group=NitroDromeScreens, parameterName=PP_Scanline2_Amount}
    Scanline Speed
    "Scanlines Thickness (Bias)"
    Scanlines Dark Value
    Scanlines Light Value
    Block Opacity
    UVs=NamedRerouteUsage()
    Time
    Block Repeat
   ) {materialFunction=/Game/Effects/Fort_Effects/Effects/Atlas/Materials/MF_Atlas_Scanlines_01}.Result
   B=Multiply(B=ScalarParameter() {defaultValue=1, group=NitroDromeScreens, parameterName=PP_AccentScanlightBrightness}) {constB=3}
  )
  n61 = Constant() {r=1}
  NamedRerouteDeclaration(None=LinearInterpolate(
    B=MaterialFunctionCall!(Default=@n60, Low/Mobile=@n61) {materialFunction=/Game/Effects/Fort_Effects/Materials/_Functions/MF_LowMobile_PathingSwitch}.Result
    Alpha=ScalarParameter() {defaultValue=0.3, group=NitroDromeScreens, parameterName=PP_Scanline2_Intensity}
   ) {constA=1}) {name=Scanlines, nodeColor=<rgba 0.4062 0 1 1>}
  NamedRerouteDeclaration() {name=Debug, nodeColor=<rgba 1 0 0.4219 1>}
  n62 = ComponentMask(None=NamedRerouteUsage()) {g=true}
  NamedRerouteDeclaration(None=Step(Y=@n37.R, X=OneMinus(None=@n62)) {constY=1}) {name=YMask, nodeColor=<rgba 1 0 0.2578 1>}
  NamedRerouteDeclaration(None=ComponentMask(None=@n4) {g=true}) {name=V, nodeColor=<rgba 0 1 0.9766 1>}
  NamedRerouteDeclaration(None=StaticSwitchParameter(True=TextureCoordinate() {coordinateIndex=1}, False=TextureCoordinate()) {group=NitroDromeScreens, parameterName=UseUV1}) {name=UVs, nodeColor=<rgba 1 0.8438 0 1>}
  n63 = NamedRerouteUsage()
  n64 = Multiply(A=@n63, B=Constant2Vector() {g=10})
  n65 = AppendVector(A=Constant() {r=0.5}, B=ScalarParameter() {defaultValue=5, group=HolographicTexture_Glitch, parameterName=StretchedScrollingSpeed})
  n66 = TextureSample(UVs=Multiply(A=Constant2Vector() {g=0.1, r=1}, B=Panner(Coordinate=@n64, Speed=@n65) {speedX=0.5, speedY=5})) {
   samplerSource=SSM_Wrap_WorldGroupSettings
   samplerType=SAMPLERTYPE_LinearColor
   texture=/Game/Environments/Asteria/Sets/NeonCity/Techart/Holograms/GlitchMap_M
  }
  n67 = SmoothStep(Value=@n66.R) {constMax=0.1}
  n68 = Constant2Vector() {g=7, r=1}
  n69 = Floor(None=Multiply(A=Time()) {constB=3})
  n70 = TextureSample(UVs=Add(A=Multiply(A=@n63, B=@n68) {constB=5}, B=Divide(A=@n69) {constB=3})) {
   samplerSource=SSM_Wrap_WorldGroupSettings
   samplerType=SAMPLERTYPE_LinearColor
   texture=&GlitchMap_M
  }
  n71 = Saturate(None=Add(A=@n70.G) {constB=0.3})
  n72 = ScalarParameter() {defaultValue=0.1, group=HolographicTexture_Glitch, parameterName=StretchedGlitchIntensity}
  n73 = NamedRerouteUsage()
  n74 = Multiply(A=Constant2Vector() {r=1}, B=Multiply(A=Multiply(A=@n67, B=@n71), B=LinearInterpolate(B=@n72, Alpha=@n73) {constB=0.1}) {constB=0.2})
  n75 = ScalarParameter() {defaultValue=0.1, group=HolographicTexture_Glitch, parameterName=StretchedGlitchIntensity}
  n76 = LinearInterpolate(B=AppendVector(A=Multiply(B=@n75) {constA=-0.75}, B=Constant()), Alpha=NamedRerouteUsage()) {constB=0.1}
  NamedRerouteDeclaration(None=Add(A=NamedRerouteUsage(), B=Add(A=@n74, B=@n76))) {name=UVs_HoriztonalGlitch, nodeColor=<rgba 1 0.2812 0 1>}
  n77 = ComponentMask(None=VectorParameter() {group=NitroDromeScreens, parameterName=Screen_LCD_Tiling, defaultValue=<rgba 150 50 0 1>}.RGB) {g=true, r=true}
  n78 = MaterialFunctionCall(UVs=@n31, Texture Scale=@n32) {materialFunction=&ScaleUVsByCenter}
  n79 = Floor(None=Multiply(A=@n77, B=@n78.UVs))
  NamedRerouteDeclaration(None=Frac(None=Divide(A=@n79, B=@n77))) {name=UVs_beforeGlitch, nodeColor=<rgba 0.3125 1 0 1>}
  n80 = Constant() {r=1}
  n81 = Time()
  n82 = LinearInterpolate(
   A=ScalarParameter() {group=NitroDromeScreens, parameterName=GlitchOffset}
   B=Multiply(A=VertexColor().R) {constB=5}
   Alpha=ScalarParameter() {group=NitroDromeScreens, parameterName=UseVertexColorGlitchOffset}
  )
  n83 = Sine(None=Multiply(
    A=Add(A=@n81, B=@n82)
    B=ScalarParameter() {defaultValue=0.2, group=HolographicTexture_Glitch, parameterName=GlitchFrequency, sortPriority=10}
   ) {constB=0.2})
  n84 = SmoothStep(
   Min=Subtract(A=@n80, B=ScalarParameter() {defaultValue=0.075, group=HolographicTexture_Glitch, parameterName=GlitchDuration, sortPriority=10})
   Max=@n80
   Value=Add(A=Multiply(A=@n83) {constB=0.5}) {constB=0.5}
  ) {constMax=0.9, constMin=0.8}
  n85 = MaterialFunctionCall() {materialFunction=/Game/MF__FloatFlickerRandom_Generic}
  n86 = NamedRerouteUsage()
  n87 = Saturate(None=Divide(A=Subtract(A=@n86) {constB=9}))
  n88 = Constant()
  NamedRerouteDeclaration(None=LinearInterpolate(
    A=Multiply(A=@n84, B=@n85.RandomFlickerFloat)
    Alpha=If(A=@n87, "A > B"=@n88, "A == B"=@n88, "A < B"=@n87) {constB=1, desc=Glitch when about to idle}
   )) {name=GlitchFrequencyMask, nodeColor=<rgba 0.9377 1 0.8953 1>}
  n89 = Multiply(A=@n3) {constB=4}
  n90 = Fresnel() {baseReflectFraction=0, exponent=1}
  NamedRerouteDeclaration(None=Multiply(A=RayTracingQualitySwitch(Normal=@n3, RayTraced=@n89), B=OneMinus(None=@n90))) {name=Emissive_OUT, nodeColor=<rgba 0 0.6953 1 1>}
  n91 = VectorParameter() {group=NitroDromeScreens, parameterName=Timer_Color_BG, sortPriority=200, defaultValue=<rgba 0 0.2586 0.9844 1>}
  n92 = VectorParameter() {group=NitroDromeScreens, parameterName=Timer_Color_Digits, sortPriority=200, defaultValue=<rgba 18.7852 20 19.3731 1>}
  NamedRerouteDeclaration(None=LinearInterpolate(A=@n91.RGB, B=@n92.RGB, Alpha=NamedRerouteUsage())) {name=TimerColor, nodeColor=<rgba 0.1953 0 1 1>}
  NamedRerouteDeclaration(None=NamedRerouteUsage()) {name=UVs_Quantized, nodeColor=<rgba 0.125 1 0 1>}
  NamedRerouteDeclaration(None=TextureSample(UVs=Multiply(A=@n78.UVs, B=@n77)) {samplerType=SAMPLERTYPE_Masks, texture=/Game/Environments/Asteria/Sets/NeonCity/Techart/Textures/T_LCD_Panels_LCDSoft}.RGB) {name=QuantizedMask, nodeColor=<rgba 1 0 0.4453 1>}
  n93 = NamedRerouteUsage()
  NamedRerouteDeclaration(None=DDX(Value=@n93)) {name=TimerDDX, nodeColor=<rgba 1 0.9609 0 1>}
  NamedRerouteDeclaration(None=StaticBoolParameter() {group=NitroDromeScreens, parameterName=UseTimer}) {name=UseTimer, nodeColor=<rgba 1 0 0.8437 1>}
  NamedRerouteDeclaration(None=DDY(Value=@n93)) {name=TimerDDY, nodeColor=<rgba 1 0.9609 0 1>}
  NamedRerouteUsage()
  NamedRerouteUsage()
  NamedRerouteUsage()
  TextureCoordinate()
 ]}
